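# Translucent unlit sheen material for Niagara mesh particles on the Hacivat Pattern pickaxe, sweeping a highlight across the mesh.
# Dynamic parameters drive gradient panning, rotation, fill and vertex-alpha contrast.
# Static switches choose a rotatable particle-colored highlight, a noise-tinted HDR sheen, or a sphere-mask sweep remapped through a borrowed Candy Apple curve atlas row.
# Particle color feeds emissive and opacity, optionally masked by vertex alpha, with emissive passing through shared FX adjustments.
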Material /BRCosmetics/Weapons/FORT_Melee/Pickaxe_Hacivat_Pattern/FX/M_Hasivat_Pattern_Sheen
graph {
 properties {
  bUsedWithNiagaraMeshParticles=true
  blendMode=BLEND_Translucent
  pixelDepthOffsetMode=PDOM_Legacy
  previewMesh=/BRCosmetics/Weapons/FORT_Melee/Pickaxe_Crazy_Eight_Tech_Male_1H/FX/Crazy_Eight_Tech_Male_1h_R_Sheen
  shadingModel=MSM_Unlit
  translucencyPass=MTP_BeforeDOF
  usedShadingModels=Unlit
 }
 nodes [
  n1 = Multiply(A=ParticleColor().RGB, B=DynamicParameter() {defaultValue=<rgba 0 1 1 0>, paramNames [Sweep, UV_Rotation, Fill, VertColorContrast]}.Fill)
  n2 = MaterialFunctionCall(
   In=VertexColor().A
   Contrast=DynamicParameter() {defaultValue=<rgba 0 1 1 0>, paramNames [Sweep, UV_Rotation, Fill, VertColorContrast]}.VertColorContrast
  ) {materialFunction=&CheapContrast}
  n3 = MaterialFunctionCall(UVs=TextureCoordinate(), "Rotation Angle (0-1)"=ScalarParameter() {parameterName=Gradient Angle}) {materialFunction=&CustomRotator}
  n4 = Add(A=@n3.Rotated Values, B=Constant2Vector() {g=-0.45})
  n5 = Clamp(None=@n4)
  n6 = ScalarParameter() {parameterName=UV Clamp?}
  n7 = DynamicParameter() {defaultValue=<rgba 0 1 1 0>, paramNames [Sweep, UV_Rotation, Fill, VertColorContrast]}
  n8 = TextureSampleParameter2D(UVs=Panner(Coordinate=LinearInterpolate(A=@n4, B=@n5, Alpha=@n6), Time=@n7.Sweep) {speedY=0.5}) {parameterName=Texture Gradient, texture=/Game/Effects/Fort_Effects/Textures/Grad/gradient_healed}
  n9 = ScalarParameter() {defaultValue=2, parameterName=Gradient Multiply}
  n10 = Power(
   Base=Clamp(None=Multiply(A=@n8.R, B=@n9) {constB=2}, Max=ScalarParameter() {defaultValue=0.5, parameterName=Gradient Clamp Max}) {maxDefault=0.5}
   Exp=ScalarParameter() {defaultValue=1, parameterName=Power Exponent}
  )
  n11 = Panner(Coordinate=Add(A=@n3.Rotated Values, B=Constant2Vector() {g=-0.44}), Time=@n7.Sweep) {speedY=0.5}
  n12 = Round(None=Multiply(
    A=TextureSample(UVs=@n11) {samplerSource=SSM_Clamp_WorldGroupSettings, texture=&gradient_healed}.R
    B=ScalarParameter() {defaultValue=1.5, parameterName=Gradient Width}
   ) {constB=1.5})
  n13 = TextureCoordinate() {vTiling=0.5}
  n14 = Constant2Vector()
  n15 = DynamicParameter() {defaultValue=<rgba 0 1 1 0>, paramNames [Sweep, UV_Rotation, Fill, VertColorContrast]}
  n16 = TextureSampleParameter2D(UVs=Panner(Coordinate=Add(A=@n13, B=@n14), Time=@n15.Sweep) {speedY=0.5}) {parameterName=Texture Gradient, texture=&gradient_healed}
  n17 = Panner(Coordinate=Add(A=TextureCoordinate() {uTiling=0}, B=Constant2Vector() {g=-0.45}), Time=@n15.Sweep) {speedY=1}
  n18 = Add(
   A=Clamp(None=Multiply(A=@n16.RGB) {constB=2}) {maxDefault=0.5}
   B=Multiply(A=TextureSample(UVs=@n17) {texture=/Game/Global/Textures/Default/Masks_and_Gradients/16x}.RGB) {constB=0.5}
  ) {constB=0}
  n19 = StaticSwitchParameter(True=Multiply(A=ParticleColor().A, B=Add(A=@n10, B=@n12)), False=Multiply(A=ParticleColor().A, B=Clamp(None=@n18))) {parameterName=Use Simple highlight}
  Add(
   A=Multiply(A=@n1, B=@n2.Result)
   B=StaticSwitchParameter(True=Multiply(A=@n2.Result, B=@n19), False=@n19) {parameterName=Use VertexColor Alpha}
  )
  n20 = DynamicParameter() {defaultValue=<rgba 0 1 1 0>, paramNames [Sweep, UV_Rotation, Fill, VertColorContrast]}
  n21 = LinearInterpolate(Alpha=@n20.Sweep) {constA=0.5, constB=-0.5}
  n22 = TextureCoordinate()
  n23 = ComponentMask(None=Add(
    A=@n21
    B=MaterialFunctionCall(UVs=@n22, "Rotation Angle (0-1)"=@n20.UV_Rotation) {materialFunction=&CustomRotator}.Rotated Values
   )) {r=true}
  n24 = CurveAtlasRowParameter(CurveTime=@n23) {
   atlas=&AthenaCurveAtlas01
   curve=/Game/Accessories/FORT_Backpacks/Backpack_M_MED_Candy_Apple/FX/Curve_CandyApple_Backpack
   defaultValue=226
  }
  n25 = CurveAtlasRowParameter(CurveTime=SphereMask(A=TextureCoordinate(), B=Constant() {r=0.5}) {attenuationRadius=0.51, hardnessPercent=0}) {
   atlas=&AthenaCurveAtlas01
   curve=&Curve_CandyApple_Backpack
   defaultValue=226
  }
  n26 = CurveAtlasRowParameter(CurveTime=LinearInterpolate(A=@n23, B=Add(A=@n21, B=@n24.R), Alpha=@n25.B)) {
   atlas=&AthenaCurveAtlas01
   curve=&Curve_CandyApple_Backpack
   defaultValue=226
  }
  n27 = Add(A=Multiply(A=@n26.G, B=ParticleColor().RGB))
  n28 = Multiply(
   A=LinearInterpolate(
    A=VectorParameter() {parameterName=Color_A, defaultValue=<rgba 0.6469 0.7188 0.6799 1>}.RGB
    B=VectorParameter() {parameterName=Color_B, defaultValue=<rgba 0.5 0.7657 1 1>}.RGB
    Alpha=TextureSample() {texture=/Game/Global/Textures/Default/Noise/T_FireNoise_02}.R
   )
   B=@n18
  )
  n29 = Clamp(None=@n28)
  n30 = Multiply(A=ComponentMask(None=@n18) {g=true}, B=ScalarParameter() {defaultValue=2, parameterName=HDR_INtensity})
  n31 = StaticSwitchParameter(
   True=ParticleColor().RGB
   False=Add(
    A=MaterialFunctionCall(LDR Input=@n29, HDR Tint=@n28, HDR Intensity Multiplier=@n30) {materialFunction=&DeriveHDRfromLDR}.RemappedRange
    B=Multiply(A=@n18) {constB=0.25}
   )
  ) {parameterName=Use Simple highlight}
  MaterialFunctionCall(Emissive=Add(A=@n1, B=StaticSwitchParameter(True=@n27, False=@n31) {parameterName=Use_Sphere_Sweep})) {materialFunction=&MF_FX_EmissiveAdjustments}
  n32 = Constant() {r=0.1}
  Multiply(A=Subtract(A=MaterialFunctionCall(Speed=@n32) {materialFunction=&DebugTimeSine}.Result) {constB=0.5})
  Multiply()
  ScalarParameter() {parameterName=VertexColorCheapContrast}
  VectorParameter() {defaultValue=<rgba 0.6823 0.8441 1 1>}
 ]
}
