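# Niagara idle effect for the Hacivat Pattern pickaxe, playing an animated sheen with moving flares and a brief flame puff.
# SheenMesh_Loop bursts one particle feeding a mesh and two sprite renderers, with material parameters and custom flare position, size and color attributes animated by curves.
# Flame_Center pulses SubUV flame sprites along a spawn-position curve keyed to emitter looped age.
# User.HueColor tints the sheen and flares, while the flame uses its own color ramp.
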NiagaraSystem /BRCosmetics/Weapons/FORT_Melee/Pickaxe_Hacivat_Pattern/FX/NS_Hasivat_Pattern_Idle
dependencies {
 usedDataInterfaces [&NiagaraDataInterfaceCurve, &NiagaraDataInterfaceVectorCurve, &NiagaraDataInterfaceColorCurve]
 usedDynamicInputs [&Multiply_Vector2, &RandomRangeVector2D]
 usedModules [
  &self:SystemSpawnScript.SystemScriptSource.SystemScriptGraph.NiagaraNodeAssignment_2.SetVariables_#1
  &SystemState
 ]
}
emitters {
 Flame_Center {
  emitterUpdateScript [
   &EmitterState()
   &SpawnRate(
    SpawnRate=&FloatFromCurve(
     FloatCurve=DataInterface() {propertyValues {Curve {keys [
         {interpMode=RCIM_Cubic, tangentMode=RCTM_User, time=0.1359, value=0}
         {interpMode=RCIM_Cubic, tangentMode=RCTM_User, time=0.1859, value=1.1333}
         <at 0.3359 0 RCIM_Cubic>
        ]}}}
     CurveIndex=$Emitter.LoopedAge
     Scale Curve=40
    )
    Use Spawn Probability=false
   )
  ]
  particleSpawnScript [
   /Niagara/Modules/Spawn/Initialization/InitializeParticle(
    Color=<rgba 1 1 1 5>
    Lifetime=&RandomRangeFloat(Minimum=0.6, Maximum=1.5)
    Module.Write Color=true
    Module.Write Lifetime=true
    Module.Write Mass=true
    Module.Write Material Random=true
    Module.Write Position=true
    Module.Write Scale=false
    Module.Write SpriteRotation=true
    Module.Write SpriteSize=true
    Position=&VectorFromCurve(
     VectorCurve=DataInterface() {propertyValues {
       XCurve {keys []}
       YCurve {keys [
         {interpMode=RCIM_Cubic, tangentMode=RCTM_User, time=0.122, value=-4.9315}
         {arriveTangent=349.5537, interpMode=RCIM_Cubic, leaveTangent=349.5537, tangentMode=RCTM_User, time=0.2086, value=18.6205}
         {arriveTangent=169.1147, interpMode=RCIM_Cubic, leaveTangent=169.1147, tangentMode=RCTM_User, time=0.3168, value=33.2752}
        ]}
       ZCurve {keys [
         {arriveTangent=92.5363, interpMode=RCIM_Cubic, leaveTangent=92.5363, tangentMode=RCTM_User, time=0.1194, value=50.483}
         {tangentMode=RCTM_User, time=0.3278, value=42.7274}
        ]}
      }}
     CurveIndex=$Emitter.LoopedAge
    )
    Sprite Rotation=&RandomRangeFloat(Minimum=0, Maximum=360)
    Sprite Size=&Vector2DFromFloat(Value=&RandomRangeFloat(Minimum=30, Maximum=40))
   )
   &AddVelocity(Coordinate Space=::World, Velocity=<0 0 5>)
  ]
  particleUpdateScript [
   &ParticleState()
   &Drag(Drag=2, Use Linear Drag=true, Use Rotational Drag=false)
   &SolveForcesAndVelocity(Clamp Velocity=false, Limit Acceleration=false)
   &SubUVAnimation(
    End Frame:NiagaraInt32=&RandomRangeInt(Minimum=50, Maximum=63)
    Start Frame:NiagaraInt32=&RandomRangeInt(Minimum=0, Maximum=10)
    SubUV Loop Count:NiagaraFloat=&RandomRangeFloat(Minimum=0.8, Maximum=1)
   )
   &DynamicMaterialParameters(
    Index 0 Param 1=&FloatFromCurve(FloatCurve=DataInterface() {propertyValues {Curve {keys [{arriveTangent=2.7551, interpMode=RCIM_Cubic, leaveTangent=2.7551, tangentMode=RCTM_User, time=0, value=1}, <at 0.2 1>]}}})
    Index 0 Param 2=&RandomRangeFloat(Minimum=1, Maximum=5)
    Index 0 Param 3=&FloatFromCurve(FloatCurve=DataInterface() {propertyValues {Curve {keys [
         {arriveTangent=4.1785, interpMode=RCIM_Cubic, leaveTangent=4.1785, tangentMode=RCTM_User, time=0, value=0}
         {arriveTangent=0.9995, interpMode=RCIM_Cubic, leaveTangent=0.9995, tangentMode=RCTM_User, time=0.2223, value=0.6825}
         {arriveTangent=-0.0405, interpMode=RCIM_Cubic, leaveTangent=-0.0405, tangentMode=RCTM_User, time=1, value=1}
        ]}}})
    Index 0 Param 4=3
    Module.Param0WriteEnabled=true
    Module.Param1WriteEnabled=true
    Module.Param2WriteEnabled=true
    Module.Param3WriteEnabled=true
   ) {module=DynamicMaterialParameters001}
   &ScaleColor(
    Scale Alpha=&Multiply_Float(
     A=&FloatFromCurve(
      FloatCurve=DataInterface() {propertyValues {Curve {keys [
          {interpMode=RCIM_Cubic, tangentMode=RCTM_User, time=0, value=1}
          {arriveTangent=-1, interpMode=RCIM_Cubic, leaveTangent=-1, time=0.3, value=0.25}
          {interpMode=RCIM_Cubic, tangentMode=RCTM_User, time=1, value=0}
         ]}}}
      Scale Curve=0.1
     )
     B=1
    )
    Scale RGB=&MakeVectorFromLinearColorRGB(Color=&ColorFromCurve(ColorCurve=DataInterface() {propertyValues {
        AlphaCurve {keys []}
        BlueCurve {keys [{arriveTangent=-1.0913, interpMode=RCIM_Cubic, time=0.2533, value=1.1305}, <at 0.75 0.02 RCIM_Cubic>]}
        GreenCurve {keys [{arriveTangent=-1.8048, interpMode=RCIM_Cubic, time=0.2533, value=2.6088}, <at 0.75 0.02 RCIM_Cubic>]}
        RedCurve {keys [{arriveTangent=9.6532, interpMode=RCIM_Cubic, time=0.2533, value=10}, <at 0.75 1 RCIM_Cubic>]}
       }}))
    ScaleA=true
    ScaleRGB=true
   )
   &ScaleSpriteSize(Scale Factor=&Vector2DFromFloat(Value=&FloatFromCurve(FloatCurve=DataInterface() {propertyValues {Curve {keys [<at 0 0.8>,<at 1 1>]}}})))
   &CameraOffset(Camera Offset Amount=4)
  ]
  rendererClasses [&NiagaraSpriteRendererProperties]
  renderers [{rendererClass=&NiagaraSpriteRendererProperties}]
 }
 SheenMesh_Loop {
  emitterUpdateScript [
   &EmitterState()
   &SpawnBurst_Instantaneous(Spawn Count=1, Use Spawn Probability=false)
  ]
  particleSpawnScript [
   /SetVariables(
    Particles.ColorFlare:LinearColor=&MultiplyLinearColor(Linear Color=$System.GlintColor, Scale Factor=<20 20 20 20>)
    Particles.PositionFlare:NiagaraPosition=$Particles.Position
   ) {isSetParameters=true}
   &InitializeParticle(
    Color=&HueShiftLinearColor(Hue=0, Saturation=1, Value=7, Linear Color=$User.HueColor)
    Lifetime=&RandomRangeFloat(Minimum=0.8, Maximum=0.8)
    Module.Write Color=true
    Module.Write Lifetime=true
    Module.Write Mass=true
    Module.Write Material Random=true
    Module.Write Position=false
    Module.Write Scale=true
    Module.Write SpriteRotation=false
    Module.Write SpriteSize=true
    Position=<0 -1 -2>
    Sprite Size=<1 2>
   )
   /NiagaraFN/Sprite/FN_SpriteAlignmentAndFacing(SpriteAlignment:Vector3f=<0 0 1>, SpriteFacing:Vector3f=<1 0 0>)
  ]
  particleUpdateScript [
   &ParticleState()
   &ScaleSpriteSize(Scale Factor=&MakeVector2D(
     X=&FloatFromCurve(
      FloatCurve=DataInterface() {propertyValues {Curve {keys [
          {arriveTangent=7.953, interpMode=RCIM_Cubic, leaveTangent=7.953, tangentMode=RCTM_User, time=-0.0008, value=0.5}
          {arriveTangent=0.0662, interpMode=RCIM_Cubic, leaveTangent=0.0662, time=0.1, value=1.5834}
          {arriveTangent=-4.779, interpMode=RCIM_Cubic, leaveTangent=-4.779, time=0.2045, value=0.5136}
          {arriveTangent=-0.1355, interpMode=RCIM_Cubic, leaveTangent=-0.1355, tangentMode=RCTM_User, time=0.4, value=0.1497}
         ]}}}
      Scale Curve=10
     )
     Y=&FloatFromCurve(
      FloatCurve=DataInterface() {propertyValues {Curve {keys [
          {arriveTangent=7.953, interpMode=RCIM_Cubic, leaveTangent=7.953, tangentMode=RCTM_User, time=-0.0008, value=0.5}
          {arriveTangent=0.8674, interpMode=RCIM_Cubic, leaveTangent=0.8674, time=0.15, value=1.8998}
          {arriveTangent=-6.9225, interpMode=RCIM_Cubic, leaveTangent=-6.9225, time=0.2096, value=0.6825}
          {arriveTangent=-0.4678, interpMode=RCIM_Cubic, leaveTangent=-0.4678, tangentMode=RCTM_User, time=0.4, value=0.1692}
         ]}}}
      Scale Curve=10
     )
    ))
   /SetVariables(Particles.SpriteSizeFlare2:Vector2f=&MakeVector2D(
     X=&FloatFromCurve(
      FloatCurve=DataInterface() {propertyValues {Curve {keys [
          {arriveTangent=7.953, interpMode=RCIM_Cubic, leaveTangent=7.953, tangentMode=RCTM_User, time=-0.0008, value=0.5}
          {arriveTangent=2.3109, interpMode=RCIM_Cubic, leaveTangent=2.3109, time=0.0852, value=1.7943}
          {arriveTangent=-3.195, interpMode=RCIM_Cubic, leaveTangent=-3.195, time=0.4969, value=1.65}
          {arriveTangent=-10.7844, interpMode=RCIM_Cubic, leaveTangent=-10.7844, tangentMode=RCTM_User, time=0.6, value=0.1497}
         ]}}}
      Scale Curve=10
     )
     Y=&FloatFromCurve(
      FloatCurve=DataInterface() {propertyValues {Curve {keys [
          {arriveTangent=7.953, interpMode=RCIM_Cubic, leaveTangent=7.953, tangentMode=RCTM_User, time=-0.0008, value=0.5}
          {arriveTangent=2.3995, interpMode=RCIM_Cubic, leaveTangent=2.3995, time=0.1101, value=1.775}
          {arriveTangent=0.0774, interpMode=RCIM_Cubic, leaveTangent=0.0774, tangentMode=RCTM_User, time=0.5202, value=1.75}
          {arriveTangent=-12.3601, interpMode=RCIM_Cubic, leaveTangent=-12.3601, tangentMode=RCTM_User, time=0.6, value=0.325}
         ]}}}
      Scale Curve=15
     )
    )) {isSetParameters=true}
   &DynamicMaterialParameters(
    Index 0 Param 1=&FloatFromCurve(FloatCurve=DataInterface() {propertyValues {Curve {keys [
         {arriveTangent=5.8912, interpMode=RCIM_Cubic, leaveTangent=5.8912, tangentMode=RCTM_User, time=0.1, value=0}
         {arriveTangent=4.12, interpMode=RCIM_Cubic, leaveTangent=4.12, tangentMode=RCTM_User, time=0.7, value=1.8}
        ]}}})
    Index 0 Param 2=1
    Index 0 Param 3=&FloatFromCurve(
     FloatCurve=DataInterface() {propertyValues {Curve {keys [
         {interpMode=RCIM_Cubic, tangentMode=RCTM_User, time=0.2, value=0}
         {arriveTangent=-4.9575, interpMode=RCIM_Cubic, leaveTangent=-4.9575, tangentMode=RCTM_User, time=0.3971, value=0.2156}
         <at 1 0 RCIM_Cubic>
        ]}}}
     Scale Curve=0.2
    )
    Index 0 Param 4=&FloatFromCurve(FloatCurve=DataInterface() {propertyValues {Curve {keys [
         {interpMode=RCIM_Cubic, tangentMode=RCTM_User, time=-0.006, value=-0.3697}
         {arriveTangent=26.5624, interpMode=RCIM_Cubic, leaveTangent=26.5624, tangentMode=RCTM_User, time=0.3412, value=2.7089}
         {arriveTangent=1.6138, interpMode=RCIM_Cubic, leaveTangent=1.6138, tangentMode=RCTM_User, time=0.5387, value=4.3118}
         <at 1 4.7685>
        ]}}})
    Index 1 Param 1=1
    Index 1 Param 2=1
    Module.Param0WriteEnabled=true
    Module.Param0WriteEnabled1=true
    Module.Param1WriteEnabled=true
    Module.Param1WriteEnabled1=true
    Module.Param2WriteEnabled=true
    Module.Param2WriteEnabled1=true
    Module.Param3WriteEnabled=true
    Module.Param3WriteEnabled1=true
    Write Parameter Index 1=true
   )
   &CameraOffset()
   /SetVariables(Particles.CameraOffsetFlare:NiagaraFloat=&FloatFromCurve(FloatCurve=DataInterface() {propertyValues {Curve {keys [
         {interpMode=RCIM_Cubic, tangentMode=RCTM_User, time=0, value=0}
         {interpMode=RCIM_Cubic, tangentMode=RCTM_User, time=0.6571, value=11.9942}
         <at 0.9751 7.6695 RCIM_Cubic>
        ]}}})) {isSetParameters=true}
   /SetVariables(
    Particles.ColorFlare:LinearColor=&MakeLinearColorFromVectorAndFloat(
     "Vector (RGB)"=&Multiply_Vector(
      A=&VectorFromFloat(Value=20)
      B=&MakeVectorFromLinearColorRGB(Color=$User.HueColor)
     )
     "Float (Alpha)"=&FloatFromCurve(
      FloatCurve=DataInterface() {propertyValues {Curve {keys [
          {arriveTangent=9.0725, interpMode=RCIM_Cubic, leaveTangent=9.0725, tangentMode=RCTM_User, time=0, value=0}
          <at 0.05 0.9847 RCIM_Cubic>
          {arriveTangent=-0.0405, interpMode=RCIM_Cubic, leaveTangent=-0.0405, tangentMode=RCTM_User, time=1, value=0}
         ]}}}
      Scale Curve=5
     )
    )
    Particles.ColorFlare3:LinearColor=&MakeLinearColorFromVectorAndFloat(
     "Vector (RGB)"=&Multiply_Vector(
      A=&VectorFromFloat(Value=20)
      B=&MakeVectorFromLinearColorRGB(Color=$User.HueColor)
     )
     "Float (Alpha)"=&FloatFromCurve(FloatCurve=DataInterface() {propertyValues {Curve {keys [
          {arriveTangent=2.4852, interpMode=RCIM_Cubic, leaveTangent=2.4852, tangentMode=RCTM_User, time=0, value=0}
          <at 0.5 0.8 RCIM_Cubic>
          {arriveTangent=-0.0405, interpMode=RCIM_Cubic, leaveTangent=-0.0405, tangentMode=RCTM_User, time=1, value=0}
         ]}}})
    )
   ) {isSetParameters=true}
   /SetVariables(
    Particles.PositionFlare:Vector3f=&VectorFromCurve(VectorCurve=DataInterface() {propertyValues {
       XCurve {keys [
         <at 0.0009 -3.3 RCIM_Cubic>
         {arriveTangent=0.9114, interpMode=RCIM_Cubic, leaveTangent=0.9114, tangentMode=RCTM_User, time=0.8155, value=-2.0807}
         {arriveTangent=11.1923, interpMode=RCIM_Cubic, leaveTangent=11.1929, tangentMode=RCTM_User, time=1, value=1.5}
        ]}
       YCurve {keys [<at 0 0 RCIM_Cubic>]}
       ZCurve {keys [
         {arriveTangent=8.0617, interpMode=RCIM_Cubic, leaveTangent=8.0617, tangentMode=RCTM_User, time=0, value=15}
         {arriveTangent=104.9599, interpMode=RCIM_Cubic, leaveTangent=104.9599, tangentMode=RCTM_User, time=0.1892, value=27.259}
         {arriveTangent=18.9662, interpMode=RCIM_Cubic, leaveTangent=18.9662, tangentMode=RCTM_User, time=0.4418, value=60}
        ]}
      }})
    Particles.PositionFlare2:Vector3f=&VectorFromCurve(VectorCurve=DataInterface() {propertyValues {
       XCurve {keys [<at 0 0>]}
       YCurve {keys [
         {arriveTangent=46.1176, interpMode=RCIM_Cubic, time=0.1, value=-10.0954}
         {arriveTangent=81.2402, interpMode=RCIM_Cubic, leaveTangent=81.2402, tangentMode=RCTM_User, time=0.3756, value=10.6461}
         {arriveTangent=190.4581, interpMode=RCIM_Cubic, leaveTangent=190.4581, tangentMode=RCTM_User, time=0.698, value=65.3525}
        ]}
       ZCurve {keys [
         {arriveTangent=92.5363, interpMode=RCIM_Cubic, leaveTangent=92.5363, tangentMode=RCTM_User, time=0.0992, value=46.0418}
         {arriveTangent=-22.1561, interpMode=RCIM_Cubic, leaveTangent=-22.1561, time=0.2681, value=39.9817}
         {arriveTangent=47.6572, interpMode=RCIM_Cubic, leaveTangent=47.6572, time=0.3453, value=40.5883}
         {arriveTangent=110.6832, interpMode=RCIM_Cubic, leaveTangent=110.6832, tangentMode=RCTM_User, time=0.7021, value=60.6647}
        ]}
      }})
    Particles.PositionFlare3:Vector3f=&VectorFromCurve(VectorCurve=DataInterface() {propertyValues {
       XCurve {keys [<at 0 0>]}
       YCurve {keys [
         {arriveTangent=46.1176, interpMode=RCIM_Cubic, time=0.271, value=10}
         {arriveTangent=99.4237, interpMode=RCIM_Cubic, leaveTangent=99.4237, time=0.5112, value=14.1008}
         {arriveTangent=233.8004, interpMode=RCIM_Cubic, leaveTangent=233.8004, tangentMode=RCTM_User, time=0.7146, value=54.1008}
        ]}
       ZCurve {keys [
         {arriveTangent=93.406, interpMode=RCIM_Cubic, leaveTangent=93.406, tangentMode=RCTM_User, time=0.1, value=20}
         {arriveTangent=114.1367, interpMode=RCIM_Cubic, leaveTangent=114.1367, tangentMode=RCTM_User, time=0.6984, value=54.8472}
        ]}
      }})
   ) {isSetParameters=true}
   &SolveForcesAndVelocity(Clamp Velocity=false, Limit Acceleration=false)
   &ScaleColor(
    Scale Alpha=&FloatFromCurve(FloatCurve=DataInterface() {propertyValues {Curve {keys [
         {interpMode=RCIM_Cubic, tangentMode=RCTM_User, time=0, value=0}
         {arriveTangent=-4.695, interpMode=RCIM_Cubic, leaveTangent=-4.695, tangentMode=RCTM_User, time=0.2, value=0}
         {arriveTangent=-3.3333, interpMode=RCIM_Cubic, leaveTangent=-3.3333, tangentMode=RCTM_User, time=0.3, value=0}
         {arriveTangent=2.5, interpMode=RCIM_Cubic, leaveTangent=2.5, tangentMode=RCTM_User, time=0.5, value=0}
         {arriveTangent=2, interpMode=RCIM_Cubic, leaveTangent=2, time=0.6, value=1}
         <at 1 1 RCIM_Cubic>
        ]}}})
    ScaleA=true
    ScaleRGB=true
   )
  ]
  rendererClasses [&NiagaraSpriteRendererProperties, &NiagaraMeshRendererProperties]
  renderers [
   {rendererClass=&NiagaraSpriteRendererProperties}
   {rendererClass=&NiagaraSpriteRendererProperties, rendererIndex=1}
   {rendererClass=&NiagaraMeshRendererProperties, rendererIndex=2}
  ]
 }
}
summary {
 systemName=NS_Hasivat_Pattern_Idle
 emitters [
  {emitterName=SheenMesh_Loop, rendererClasses [&NiagaraSpriteRendererProperties, &NiagaraMeshRendererProperties]}
  {emitterName=Flame_Center, rendererClasses [&NiagaraSpriteRendererProperties]}
 ]
 userVariables [
  User.HueColor {type=LinearColor, description="", defaultValue {struct=LinearColor, value=<rgba 1 0.0833 0 1>}}
  User.Skeletal_mesh {
   type=NiagaraDataInterfaceSkeletalMesh
   defaultValue {
    struct=NiagaraExt_VariableValue_DataInterface
    value {dataInterface=&self:NiagaraDataInterfaceSkeletalMesh_1}
   }
  }
  User.Swing_SK_Mesh {
   type=NiagaraDataInterfaceSkeletalMesh
   defaultValue {
    struct=NiagaraExt_VariableValue_DataInterface
    value {dataInterface=&self:NiagaraDataInterfaceSkeletalMesh_0}
   }
  }
 ]
}
