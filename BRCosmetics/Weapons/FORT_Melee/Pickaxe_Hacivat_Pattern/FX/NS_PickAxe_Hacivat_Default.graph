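# Niagara system for the Hacivat pickaxe's default trail effect, built from two ribbon emitters.
# AirTrail spawns an anim-trail ribbon between two bone indices, dark-tinted, narrowing and fading over its short life.
# FN_Attached_Ribbon_BETA samples sockets on the user skeletal mesh after a brief delay, spawning one colour-ramped ribbon.
# A scratch-pad module flags bone-locked particles, which hold zero age and render at the live socket position.
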NiagaraSystem /BRCosmetics/Weapons/FORT_Melee/Pickaxe_Hacivat_Pattern/FX/NS_PickAxe_Hacivat_Default
dependencies {
 usedDataInterfaces [&NiagaraDataInterfaceArrayInt32, &NiagaraDataInterfaceCurve, &NiagaraDataInterfaceColorCurve]
 usedDynamicInputs []
 usedModules [&SystemState]
}
emitters {
 AirTrail {
  emitterUpdateScript [
   &EmitterState(Life Cycle Mode=::NewEnumerator1, Loop Behavior=::NewEnumerator1, Loop Duration=0.2, UseLoopDelay=false)
   &FN_SpawnAnimTrail_v2()
  ]
  particleSpawnScript [
   /Game/Effects/Niagara/Modules/AnimTrail/FN_InitializeAnimTrail/FN_InitializeAnimTrailBoneIndecies(First Bone or Socket Index=2, Second Bone or Socket Index=3)
   /Game/Effects/Niagara/Modules/AnimTrail/FN_InitializeAnimTrail/FN_InitializeAnimTrail(AnimTrailWidth=1, Lifetime=0.15, Position Offset:Vector3f=<0 0 -20>)
   /SetVariables() {isSetParameters=true}
  ]
  particleUpdateScript [
   &SolveForcesAndVelocity(Clamp Velocity=false, Limit Acceleration=false)
   &ScaleRibbonWidth(Ribbon Width Scale=&FloatFromCurve(FloatCurve=DataInterface() {propertyValues {Curve {keys [
         {arriveTangent=-1.2579, interpMode=RCIM_Cubic, time=0, value=1}
         {arriveTangent=-0.9664, interpMode=RCIM_Cubic, leaveTangent=-0.9664, tangentMode=RCTM_User, time=1, value=0}
        ]}}}))
   &FN_UpdateAnimTrail_v2()
   &DynamicMaterialParameters(
    Index 0 Param 1=&UniformRangedFloat(Minimum=0, Maximum=1)
    Index 0 Param 2=&UniformRangedFloat(Minimum=0, Maximum=1)
    Index 0 Param 3=1
    Index 0 Param 4=1
    Module.Param0WriteEnabled=true
    Module.Param1WriteEnabled=true
    Module.Param2WriteEnabled=true
    Module.Param3WriteEnabled=true
   )
   &Color(Color=&ColorFromCurve(ColorCurve=DataInterface() {propertyValues {
       AlphaCurve {keys [<at 0.4417 1 RCIM_Cubic>,<at 1 0 RCIM_Cubic>]}
       BlueCurve {keys [<at 0 0.05 RCIM_Cubic>]}
       GreenCurve {keys [<at 0 0.0708 RCIM_Cubic>]}
       RedCurve {keys [<at 0 0.1 RCIM_Cubic>]}
      }}))
  ]
  rendererClasses [&NiagaraRibbonRendererProperties]
  renderers [{rendererClass=&NiagaraRibbonRendererProperties}]
 }
 FN_Attached_Ribbon_BETA {
  emitterSpawnScript [/SetVariables(Emitter.Number of Trails=1) {isSetParameters=true}]
  emitterUpdateScript [
   &EmitterState(Life Cycle Mode=::NewEnumerator1, Loop Behavior=::NewEnumerator1, Loop Delay=0.07, Loop Duration=1, UseLoopDelay=true)
   &SpawnBurst_Instantaneous(
    Spawn Count=$Emitter.Initial.Number of Trails
    Spawn Group=-1
    Spawn Probability=HlslExpression() {hlslExpression="Emitter.LoopCount == 0"}
    Use Spawn Probability=true
   )
   &SpawnRate(
    SpawnRate=&Multiply_Float(A=60, B=&MakeFloatFromInt(Int=$Emitter.Number of Trails))
    Use Spawn Probability=false
   )
  ]
  particleSpawnScript [
   &FN_SetSpawnGroup(Number of Spawn Groups:NiagaraInt32=$Emitter.Initial.Number of Trails)
   /SetVariables(Particles.SpawnGroup=&SelectIntFromArray(Direct Array Index=$Particles.SpawnGroup, Int Selection Array=DataInterface() {propertyValues {IntData [0,1,2,3]}})) {isSetParameters=true}
   &self:FN_Attached_Ribbon_BETA.NiagaraScratchPadContainer_2.LockParticlesToBones_0() {module=LockParticlesToBones}
   &InitializeRibbon(
    Color=<rgba 1 1 1 2>
    Lifetime=0.12
    Module.Write Color=true
    Module.Write Lifetime=true
    Module.Write Mass=true
    Module.Write Material Random=true
    Module.Write Position=false
    Module.Write RibbonTwist=false
    Module.Write RibbonWidth=true
    Ribbon Width=12
   )
   &FN_FlipRibbonV() {enabled=false}
   /SetVariables() {isSetParameters=true}
   &SkeletalMeshLocation(
    Bone / Socket Index=$Particles.SpawnGroup
    Inherit Velocity=false
    Interpolate Mesh Sampling=false
    MaskSpawnGroup=false
    Mesh Sampling Type=::NewEnumerator5
    Sampled Position Offset=$Particles.Offset
    Sampled Position Offset Coordinate Space:ENiagaraMeshTransforms=::NewEnumerator3
    Skeletal Mesh=$User.SkeletalMesh
    Socket Sampling Mode=::NewEnumerator5
   )
   &AddVelocity(Velocity=<0 0 100>) {enabled=false}
  ]
  particleUpdateScript [
   &ParticleState()
   /SetVariables(Particles.Age:NiagaraFloat=&MakeCustomFloatFromBool(Bool=$Particles.Locked, True Float=0, False Float=$Particles.Age)) {isSetParameters=true}
   /SetVariables(Particles.NormalizedAge:NiagaraFloat=&MakeCustomFloatFromBool(Bool=$Particles.Locked, True Float=0, False Float=$Particles.NormalizedAge)) {isSetParameters=true}
   &SkeletalMeshLocation(
    Apply To Particle Position=false
    Bone / Socket Index=$Particles.SpawnGroup
    Inherit Velocity=false
    Interpolate Mesh Sampling=false
    MaskSpawnGroup=false
    Mesh Sampling Type=::NewEnumerator5
    Sampled Position Offset Coordinate Space:ENiagaraMeshTransforms=::NewEnumerator3
    Skeletal Mesh=$User.SkeletalMesh
    Socket Sampling Mode=::NewEnumerator5
    Use Sampled Mesh Attribute as Ribbon ID=false
   ) {module=SkeletalMeshLocation001}
   &SolveForcesAndVelocity(Clamp Velocity=false, Limit Acceleration=false)
   /SetVariables(Particles.RenderPosition:Vector3f=&MakeCustomVectorFromBool(
     Bool=$Particles.Locked
     True Vector=&Add_Vector(
      A=$Output.SkeletalMeshLocation001.SampledPosition
      B=&RotateVectorByQuaternion(Vector To Rotate=$Particles.Offset, Quaternion=$Output.SkeletalMeshLocation001.SampledRotation)
     )
     False Vector=&MakeCustomVectorFromBool(
      Bool=$Emitter.LocalSpace
      True Vector=&Add_Vector(A=$Output.SkeletalMeshLocation001.SampledPosition, B=Linked() {linkedVariable {name=Particles.Position, type=NiagaraPosition}})
      False Vector=Linked() {linkedVariable {name=Particles.Position, type=NiagaraPosition}}
     )
    )) {isSetParameters=true}
   &Color(Scale Alpha=&FloatFromCurve(
     FloatCurve=DataInterface() {propertyValues {Curve {keys [<at 0.0009 0.9773 RCIM_Cubic>, {arriveTangent=-0.9781, interpMode=RCIM_Cubic, leaveTangent=-0.9781, time=0.3498, value=0.9932}, <at 1 0>]}}}
     Scale Curve=&FloatFromCurve(
      FloatCurve=DataInterface() {propertyValues {Curve {keys [
          <at 0.1592 0.0035 RCIM_Cubic>
          {arriveTangent=-0.0523, interpMode=RCIM_Cubic, leaveTangent=-0.0523, time=0.4157, value=0.9671}
          <at 0.664 -0.0229>
         ]}}}
      CurveIndex=$Emitter.Age
     )
    )) {enabled=false}
   &FN_RibbonMaterial(
    Opacity Pan U Speed=0.9
    Scale Alpha:NiagaraFloat=&FloatFromCurve(
     FloatCurve=DataInterface() {propertyValues {Curve {keys [<at 0.0009 0.9773 RCIM_Cubic>, {arriveTangent=-0.9781, interpMode=RCIM_Cubic, leaveTangent=-0.9781, time=0.3498, value=0.9932}, <at 1 0>]}}}
     Scale Curve=&FloatFromCurve(
      FloatCurve=DataInterface() {propertyValues {Curve {keys [<at 0.0609 0 RCIM_Cubic>, {arriveTangent=0.0268, interpMode=RCIM_Cubic, leaveTangent=0.0268, time=0.1693, value=0.9588}, <at 0.2946 0.0063>]}}}
      CurveIndex=$Emitter.Age
     )
    )
   ) {enabled=false}
   &ScaleColor(
    Linear Color Curve=DataInterface() {propertyValues {
      AlphaCurve {keys [<at 0.2055 0>,<at 0.2994 1>,<at 0.349 1>,<at 1 0>]}
      BlueCurve {keys [<at 0.2338 0.02>,<at 0.3171 0.5>,<at 0.5775 0.05>]}
      GreenCurve {keys [<at 0.2338 0.02>,<at 0.3171 0.5>,<at 0.5775 0.05>]}
      RedCurve {keys [<at 0.2338 0.02>,<at 0.3171 0.5>,<at 0.5775 0.05>]}
     }}
    Scale Mode=::NewEnumerator2
   )
  ]
  rendererClasses [&NiagaraRibbonRendererProperties]
  renderers [{rendererClass=&NiagaraRibbonRendererProperties}]
 }
}
summary {
 systemName=NS_PickAxe_Hacivat_Default
 emitters [
  {emitterName=FN_Attached_Ribbon_BETA, rendererClasses [&NiagaraRibbonRendererProperties]}
  {emitterName=AirTrail, rendererClasses [&NiagaraRibbonRendererProperties]}
 ]
 userVariables [
  User.PhysicalSurface {type=NiagaraInt32}
  User.RibbonIDOffset {type=NiagaraInt32}
  User.SkeletalMesh {
   type=NiagaraDataInterfaceSkeletalMesh
   defaultValue {
    struct=NiagaraExt_VariableValue_DataInterface
    value {dataInterface=&self:NiagaraDataInterfaceSkeletalMesh_0}
   }
  }
 ]
}
