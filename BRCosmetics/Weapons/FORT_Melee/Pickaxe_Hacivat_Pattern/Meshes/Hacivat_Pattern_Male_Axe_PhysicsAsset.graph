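# Physics asset for the male axe mesh of the Hacivat Pattern pickaxe, a two-handed melee cosmetic weapon.
# A single capsule body on the Melee_TwoHanded_Handle bone serves as the collision shape for the weapon.
# The capsule is offset and rotated from the bone origin, and no constraints are defined, so there is no jointed or articulated physics.
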PhysicsAsset /BRCosmetics/Weapons/FORT_Melee/Pickaxe_Hacivat_Pattern/Meshes/Hacivat_Pattern_Male_Axe_PhysicsAsset
bodies {Melee_TwoHanded_Handle [{shapeType=Capsule, length=69.1632, radius=25.3463, center=<-0.0332 4.7895 18.3292>, rotation=<rot 0.7566 -90 0.0333>}]}
constraints []
